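; Wild Estate variant of the stock Battle Bus, with its own intro music and a localized scrolling LED message on the balloon.
; On cosmetic-capable instances, BeginPlay looks up the current language in LocImageTable and async-loads that text texture.
; A dynamic material on balloon slot 1 receives the texture, pan direction and a random message selection.
; Audio setup takes the flight loop from the active skin; EndPlay clears the material reference.
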(event EventBeginPlay
  (|CallParentFunction)
  (Networking|CanExecuteCosmeticEvents
    (:True
      (CallFunction|EventSetupMessagePanelforLocalization))))

(event EventEndPlay
  (params (EndPlayReason :type EEndPlayReason%20Enum :default "Destroyed"))
  (|CallParentFunction EndPlayReason)
  (Networking|CanExecuteCosmeticEvents
    (:True
      (CallFunction|EventCleanupMessagePanelforLocalization))))

(event EventGrabAudioReferences
  (Variables|Sound|SetLoopingSound (Athena|GetFlightLoopingSound (Variables|BattleBusCosmeticInstanceBase|GetActiveSkin)))
  (Variables|Sound|SetLoopingMusic "/WildEstate/Assets/Audio/Intro/IP_MSS_WildEstate_BattleBus_Music.IP_MSS_WildEstate_BattleBus_Music"))

(event Custom|EventSetupMessagePanelForLocalization
  (bind _returnvalue (Utilities|Internationalization|GetCurrentLanguage))
  (bind _locimagetable (Variables|Default|GetLocImageTable))
  (bind (_value _returnvalue_2) (Utilities|Map|Find _locimagetable _returnvalue))
  (bind (_textouse_2_b69e6377402e9a2d409c2ab94fc6cff1 _pandirection_8_aca50cd1495117d4a74e6080f556caf1) (Utilities|Struct|BreakWildEstateLEDMessageInfo _value))
  (if _returnvalue_2
    (if (not (Asset|IsAssetNull _textouse_2_b69e6377402e9a2d409c2ab94fc6cff1))
      (bind _object (Utilities|AsyncLoadAsset _textouse_2_b69e6377402e9a2d409c2ab94fc6cff1)
        (:Completed
          (bind _astexture_2d (Utilities|Casting|CastToTexture2D _object)
            (:then
              (bind _returnvalue_3 (Rendering|Material|CreateDynamicMaterialInstance (Variables|Default|GetBalloon) 1 (Rendering|Material|GetMaterial (Variables|Default|GetBalloon) 1) "" :!class /Script/Engine.PrimitiveComponent))
              (Variables|Default|SetDynamicBalloonMaterial _returnvalue_3)
              (Rendering|Material|SetTextureParameterValue _returnvalue_3 "Text" _astexture_2d) ; Specify texture to load / use
              (Rendering|Material|SetScalarParameterValue _returnvalue_3 "PanDirection" _pandirection_8_aca50cd1495117d4a74e6080f556caf1 :!class /Script/Engine.MaterialInstanceDynamic) ; Specify message pan direction
              (Rendering|Material|SetScalarParameterValue _returnvalue_3 "TextSelection" (Math|Random|RandomFloatinRange 0.0 1.0) :!class /Script/Engine.MaterialInstanceDynamic) ; Specify message
            )
            (:CastFailed
              (Development|PrintString "Failed to case Text asset to Texture!")))))
      (else
        (Development|PrintString "Text asset not found for B_StockBattleBus_WildEstate")))
    (else
      (Development|PrintString (Utilities|String|Append (Utilities|String|Append "Current language string (" _returnvalue) ") not found in LocImageTable!")))))

(event Custom|EventCleanupMessagePanelForLocalization
  (Variables|Default|SetDynamicBalloonMaterial 0))

(fn ConstructionScript
  (|CallParentFunction))
