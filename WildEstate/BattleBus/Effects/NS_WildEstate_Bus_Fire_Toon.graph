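# Niagara system for a stylized toon fire effect on the WildEstate Battle Bus, built from two burst-spawned emitters.
# NE_Empty initializes scaled particles for sprite and mesh renderers.
# NE_Empty001 feeds two sprite renderers and a light renderer with large orange particles, raised and camera-offset.
# Randomly sampled waveforms flicker its SpriteColor alpha, while the Scale Alpha waveform has no effect because its lerp endpoints are equal.
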NiagaraSystem /WildEstate/BattleBus/Effects/NS_WildEstate_Bus_Fire_Toon
dependencies {usedDataInterfaces [], usedDynamicInputs [], usedModules [&SystemState]}
emitters {
 NE_Empty {
  emitterUpdateScript [
   &EmitterState()
   &SpawnBurst_Instantaneous(Use Spawn Probability=false, UseLoopCountLimit=false)
   &SpawnRate(Use Spawn Probability=false) {enabled=false}
  ]
  particleSpawnScript [&InitializeParticle(Mesh Scale Mode=::NewEnumerator3, Mesh Uniform Scale=1.6, UsePositionOffset=false, Write Material Random=true) {module=InitializeParticle001}]
  particleUpdateScript [
   &ParticleState()
   &SolveForcesAndVelocity(Clamp Velocity=false, Limit Acceleration=false)
  ]
  rendererClasses [&NiagaraSpriteRendererProperties, &NiagaraMeshRendererProperties]
  renderers [
   {rendererClass=&NiagaraSpriteRendererProperties}
   {rendererClass=&NiagaraMeshRendererProperties, rendererIndex=1}
  ]
 }
 NE_Empty001 {
  emitterUpdateScript [
   &EmitterState()
   &SpawnBurst_Instantaneous(Use Spawn Probability=false, UseLoopCountLimit=false)
   &SpawnRate(SpawnRate=20, Use Spawn Probability=false) {enabled=false}
  ]
  particleSpawnScript [
   &PartitionParticles(Number Of Partitions=47)
   /Niagara/Modules/Spawn/Initialization/InitializeParticle(
    Color=<rgba 10 2.6514 0 1>
    Mesh Scale=&VectorFromFloat(Value=1.5)
    Module.Write Color=true
    Module.Write Lifetime=true
    Module.Write Mass=true
    Module.Write Material Random=true
    Module.Write Position=true
    Module.Write Scale=true
    Module.Write SpriteRotation=false
    Module.Write SpriteSize=true
    Position=&SimulationPosition()
    Sprite Size=<700 700>
   ) {enabled=false}
   /Niagara/Modules/Spawn/Initialization/V2/InitializeParticle(
    Color=<rgba 10 2.6514 0 1>
    Position Offset=<0 0 100>
    Sprite Size Mode=::NewEnumerator3
    Uniform Sprite Size=&MakeFloatFromVector2d(Vector 2D=<700 700>)
    UsePositionOffset=true
    Write Material Random=true
   ) {module=InitializeParticle001}
   &CameraOffset(Camera Offset Amount=100)
  ]
  particleUpdateScript [
   &ParticleState()
   &SolveForcesAndVelocity(Clamp Velocity=false, Limit Acceleration=false)
   &Light_Attributes(Falloff Exponent=1, Light Exponent=2, Light Radius=600, Radius=1)
   &ScaleColor(
    Scale Alpha=&MakeFloatFromVector(
     Vector=&VectorFromFloat(Value=&Lerp_Float(
       A=1
       B=1
       Alpha=&MakeFloatFromVector(
        Vector=&VectorFromFloat(Value=&Waveform(
          "Waveform [1]"=::NewEnumerator12
          "[1] Input Position"=&RandomRangeFloat(Minimum=0, Maximum=1)
          "[1] Frequency"=10
          "[1] Amplitude Min/Max"=<0 1>
         ))
        Channel=::NewEnumerator0
       )
      ))
     Channel=::NewEnumerator0
    )
    ScaleA=true
    ScaleRGB=true
   )
   /SetVariables(Particles.SpriteColor=&MakeLinearColorFromVectorAndFloat(
     "Vector (RGB)"=&MakeVectorFromLinearColorRGB(Color=<rgba 1 0.2651 0 1>)
     "Float (Alpha)"=&MakeFloatFromVector(
      Vector=&VectorFromFloat(Value=&Waveform(
        Global Amplitude Scale=0.2
        "Waveform [1]"=::NewEnumerator12
        "[1] Input Position"=&RandomRangeFloat(Minimum=0, Maximum=1)
        "[1] Frequency"=10
        "[1] Amplitude Min/Max"=<0.5 1>
        "[1] Bias"=0.5
       ))
      Channel=::NewEnumerator0
     )
    )) {isSetParameters=true}
   &DynamicMaterialParameters(
    Index 0 Param 1=3
    Index 0 Param 2=1
    Module.Index0Float0Write=true
    Module.Index0Float1Write=true
    Module.Index0Float2Write=true
    Module.Index0Float3Write=true
   )
  ]
  rendererClasses [&NiagaraSpriteRendererProperties, &NiagaraLightRendererProperties]
  renderers [
   {rendererClass=&NiagaraSpriteRendererProperties}
   {rendererClass=&NiagaraLightRendererProperties, rendererIndex=1}
   {rendererClass=&NiagaraSpriteRendererProperties, rendererIndex=2}
  ]
 }
}
summary {
 systemName=NS_WildEstate_Bus_Fire_Toon
 emitters [
  {emitterName=NE_Empty, rendererClasses [&NiagaraSpriteRendererProperties, &NiagaraMeshRendererProperties]}
  {emitterName=NE_Empty001, rendererClasses [&NiagaraSpriteRendererProperties, &NiagaraLightRendererProperties]}
 ]
 userVariables []
}
