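; Lobby battle bus actor for WildEstate that plays bus audio in step with the game phase.
; It listens to the game state's stateful phase-step event through the GameplayEventRouter and stops listening on EndPlay.
; GetReady plays an ignition sound on the static mesh, waits, then starts looping sound and music and records the game time.
; BusLocked stops only the looping music.
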(event EventEndPlay
  (params (EndPlayReason :type EEndPlayReason%20Enum :default "Destroyed"))
  (|CallParentFunction EndPlayReason)
  (bind _gamephasestepupdatedasynctask (Variables|Default|GetGamePhaseStepUpdatedAsyncTask)
    (:then
      (GameplayEventRouter|StopListeningtoStatefulEvent _gamephasestepupdatedasynctask))
    (:else)))

(event EventBeginPlay
  (|CallParentFunction)
  (bind _returnvalue (GameplayEventRouter|GetEventRouter "/Script/FortniteGame.FortGameState" self))
  (bind (_asynctaskproxy _payload) (GameplayEventRouter|StartListeningToStatefulEvent _returnvalue "/Script/FortniteGame.FortGamePhaseStepUpdatedEvent")
    (:then
      (Variables|Default|SetGamePhaseStepUpdatedAsyncTask _asynctaskproxy))
    (:OnEventReceived
      _)
    (:OnSavedEventStateExists
      _))
  (CallFunction|OnGamePhaseStepChanged (Utilities|Struct|BreakFortGamePhaseStepUpdatedEvent _payload)))

(event Collision|EventActorBeginOverlap
  (params (OtherActor :type Actor%20Object%20Reference)))

(event Custom|On%20Game%20Phase%20Step%20Changed
  (params (GamePhaseStep :type EAthenaGamePhaseStep%20Enum))
  (switch Utilities|FlowControl|Switch|SwitchonEAthenaGamePhaseStep GamePhaseStep
    (:GetReady
      (Audio|SpawnSoundAttached (Variables|Default|GetIgnitionSound) (Variables|Debug|GetStaticMeshComponent))
      (Utilities|FlowControl|Delay (Variables|Default|GetIgnitionTime))
      (Utilities|FlowControl|Sequence
        (:then_0
          (bind _returnvalue (Audio|SpawnSoundAttached (Variables|Default|GetLoopingSound) (Variables|Debug|GetStaticMeshComponent)))
          (Variables|Default|SetLoopingSound_Inst _returnvalue))
        (:then_1
          (bind _returnvalue_2 (Audio|SpawnSoundAttached (Variables|Default|GetLoopingMusic) (Variables|Debug|GetStaticMeshComponent)))
          (Variables|Default|SetLoopingMusic_Inst _returnvalue_2))
        (:then_2
          (Variables|Default|SetGetReadyGameTime (Utilities|Time|GetGameTimeinSeconds)))))
    (:BusLocked
      (Audio|Components|Audio|Stop (Variables|Default|GetLoopingMusic_Inst)))))

(fn ConstructionScript
  (|CallParentFunction))
