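; WildEstateAircraft is a Blueprint for the Battle Bus aircraft in WildEstate.
; It declares the event for playing effects when a player jumps from the aircraft and a construction script, both with no logic shown.
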(event Aircraft|EventPlayEffectsForPlayerJumped)

(fn ConstructionScript)
